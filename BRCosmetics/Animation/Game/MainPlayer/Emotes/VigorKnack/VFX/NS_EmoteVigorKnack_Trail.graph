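# Niagara ribbon trail effect for the VigorKnack emote, drawing two flame-like animation trails between bone or socket pairs.
# Trail_Flames sets time-of-day color, then shifts it over life through a color curve, fading alpha scaled by User.Alpha.
# Owner velocity above a threshold shortens trail lifetime and applies partial movement inheritance, while ribbon facing follows owner rotation.
# A disabled debug sprite emitter feeds owner speed into a dynamic material parameter.
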NiagaraSystem /BRCosmetics/Animation/Game/MainPlayer/Emotes/VigorKnack/VFX/NS_EmoteVigorKnack_Trail
dependencies {
 usedDataInterfaces [&NiagaraDataInterfaceColorCurve, &NiagaraDataInterfaceCurve, &NiagaraDataInterfaceSocketReader]
 usedDynamicInputs [&SimulationPosition]
 usedModules [
  &self:SystemSpawnScript.SystemScriptSource.SystemScriptGraph.NiagaraNodeAssignment_0.SetVariables_#1
  &self:SystemSpawnScript.SystemScriptSource.SystemScriptGraph.NiagaraNodeAssignment_1.SetVariables_#2
  &SystemState
 ]
}
emitters {
 FN_Debug {
  bEnabled=false
  emitterUpdateScript [
   &EmitterState()
   &SpawnBurst_Instantaneous(Spawn Count=1, Use Spawn Probability=false)
  ]
  particleSpawnScript [&InitializeParticle(Lifetime=5, Position Offset=<100 0 100>, UsePositionOffset=true, Write Material Random=true) {module=InitializeParticle001}]
  particleUpdateScript [
   &ParticleState()
   /SetVariables(Particles.Debug:NiagaraFloat=&VectorLength(Vector=$User.OwnerLocalVelocity)) {isSetParameters=true}
   &DynamicMaterialParameters(
    Index 0 Param 1=$Particles.Debug
    Module.Param0WriteEnabled=true
    Module.Param1WriteEnabled=true
    Module.Param2WriteEnabled=true
    Module.Param3WriteEnabled=true
   )
   /SetVariables(Particles.pos:NiagaraPosition=&AddVectorToPosition(Position=<0 0 0>, Vector=<100 0 100>)) {isSetParameters=true}
  ]
  rendererClasses [&NiagaraSpriteRendererProperties]
  renderers [{rendererClass=&NiagaraSpriteRendererProperties}]
 }
 Trail_Flames {
  emitterSpawnScript [
   /SetVariables(Emitter.Rand:NiagaraFloat=&RandomRangeFloat(Minimum=0, Maximum=1)) {isSetParameters=true}
   /SetVariables(Emitter.Rand2:NiagaraFloat=&RandomRangeFloat(Minimum=0, Maximum=1)) {isSetParameters=true}
  ]
  emitterUpdateScript [
   &EmitterState()
   &FN_SpawnAnimTrail(Number of Anim Trails=2)
  ]
  particleSpawnScript [
   &FN_InitializeAnimTrailBoneIndecies()
   &FN_InitializeAnimTrailBoneIndecies(AnimTrail Selection=1, First Bone or Socket Index=1, Second Bone or Socket Index=2) {module=FN_InitializeAnimTrailBoneIndecies001}
   &FN_InitializeAnimTrail(
    AnimTrailWidth=1.5
    Color=&FNTimeOfDayColor(
     Night Color=<rgba 0.7083 0.7083 0.7083 1>
     Morning Color=<rgba 0.8333 0.8333 0.8333 1>
     Day Color=<rgba 1 1 1 1>
     Evening Color=<rgba 0.8333 0.8333 0.8333 1>
    )
    Lifetime=&MakeCustomFloatFromBool(
     Bool=&SetBoolByFloatComparison(A=&VectorLength(Vector=$User.OwnerLocalVelocity), B=100)
     True Float=0.15
     False Float=0.27
    )
   )
   &DynamicMaterialParameters(
    Index 0 Param 1=&MaskFloatBySpawnGroup(Masked Float=$Emitter.Rand, Passthrough Float=$Emitter.Rand2)
    Module.Index0Float0Write=false
    Module.Index0Float1Write=true
    Module.Index0Float2Write=true
    Module.Index0Float3Write=true
   ) {enabled=false}
  ]
  particleUpdateScript [
   &ScaleColor(
    Scale Alpha=&Multiply_Float(
     A=$User.Alpha
     B=&FloatFromCurve(FloatCurve=DataInterface() {propertyValues {Curve {
         defaultValue=1
         keys [{interpMode=RCIM_Cubic, tangentMode=RCTM_User, time=0, value=1}, {interpMode=RCIM_Cubic, tangentMode=RCTM_User, time=1, value=0}]
        }}})
    )
    Scale RGB=&MakeVectorFromLinearColorRGB(Color=&MultiplyLinearColor(
      Linear Color=&ColorFromCurve(ColorCurve=DataInterface() {propertyValues {
         AlphaCurve {keys []}
         BlueCurve {keys [<at 0.0087 2 RCIM_Cubic>,<at 0.079 0.3903>,<at 0.1324 0.1507>,<at 0.1729 0.5914>,<at 0.1932 1>,<at 0.2401 1>,<at 0.2945 1>]}
         GreenCurve {keys [<at 0.0087 2 RCIM_Cubic>,<at 0.079 2>,<at 0.1324 1>,<at 0.1729 1>,<at 0.1932 0.5864>,<at 0.2401 0.1846>,<at 0.2945 0.0476>]}
         RedCurve {keys [<at 0.0087 2 RCIM_Cubic>,<at 0.079 0.3322>,<at 0.1324 0.1135>,<at 0.1729 0.4362>,<at 0.1932 0.4841>,<at 0.2401 0.4337>,<at 0.2945 0.4837>]}
        }})
      Scale Factor=&MakeVector4_FromLinearColor(LinearColor=<rgba 0.5363 0.574 0.5174 1>)
     ))
    ScaleA=true
    ScaleRGB=true
   )
   &FN_UpdateAnimTrail()
   /SetVariables(Particles.RibbonFace:Vector3f=&RotateVectorByQuaternion(Vector To Rotate=<0 0 1>, Quaternion=$Engine.Owner.Rotation)) {isSetParameters=true}
   &Drag(Drag=5, Use Linear Drag=true, Use Rotational Drag=true)
   &InheritSourceMovement(
    Applied Position Scale=&MakeCustomVectorFromBool(
     Bool=&SetBoolByFloatComparison(A=&VectorLength(Vector=$User.OwnerLocalVelocity), B=100)
     True Vector=<0.6 0.6 0.6>
     False Vector=<0 0 0>
    )
    ApplyForce=false
    ApplyPosition=true
    ApplyPositionFalloff=false
    Source Velocity:Vector3f=$User.OwnerLocalVelocity
   )
   &SolveForcesAndVelocity(Clamp Velocity=false, Limit Acceleration=false)
  ]
  rendererClasses [&NiagaraRibbonRendererProperties]
  renderers [{rendererClass=&NiagaraRibbonRendererProperties}]
 }
}
summary {
 systemName=NS_EmoteVigorKnack_Trail
 emitters [
  {emitterName=Trail_Flames, rendererClasses [&NiagaraRibbonRendererProperties]}
  {bEnabled=false, emitterName=FN_Debug, rendererClasses [&NiagaraSpriteRendererProperties]}
 ]
 userVariables [
  User.Alpha {type=NiagaraFloat, defaultValue=0, description=""}
  User.OwnerLocalVelocity {type=Vector3f, description="", defaultValue {struct=Vector3f, value=<0 0 0>}}
  User.RibbonIDOffset {type=NiagaraInt32}
 ]
}
