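# Niagara impact burst VFX for the VigorKnack emote, a short green and purple flash with sparks and streaks.
# One flare particle feeds three sprite and two mesh renderers through separate glow, corona and impact size and color attributes, scaled by system size and time-of-day color values.
# Sparks burst in a cone with gravity, drag and speed-based sizing.
# Streaks stretch and fade through curves, material parameters and time-of-day alpha scaling.
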NiagaraSystem /BRCosmetics/Animation/Game/MainPlayer/Emotes/VigorKnack/VFX/NS_Emote_VigorKnack_Impact
dependencies {
 usedDataInterfaces [
  &NiagaraDataInterfaceArrayFloat
  &NiagaraDataInterfaceCurve
  &NiagaraDataInterfaceVector2DCurve
  &NiagaraDataInterfaceColorCurve
  &NiagaraDataInterfaceArrayColor
 ]
 usedDynamicInputs [&SelectFloatFromArray, &SpawnGroup, &Lerp_Float]
 usedModules [
  &self:SystemSpawnScript.SystemScriptSource.SystemScriptGraph.NiagaraNodeAssignment_0.SetVariables_#1
  &self:SystemSpawnScript.SystemScriptSource.SystemScriptGraph.NiagaraNodeAssignment_1.SetVariables_#2
  &SystemState
 ]
}
emitters {
 NE_Impact_Flare_and_Corona {
  emitterUpdateScript [
   &EmitterState()
   &SpawnBurst_Instantaneous(Spawn Count=1, Use Spawn Probability=false)
  ]
  particleSpawnScript [
   &InitializeParticle(
    Color=<rgba 0.6387 1 0.786 1>
    Lifetime=0.19
    Mesh Scale Mode=::NewEnumerator3
    Mesh Uniform Scale=1.75
    Sprite Size Mode=::NewEnumerator3
    Uniform Sprite Size=&Multiply_Float(A=245, B=$System.ModifySize)
    UsePositionOffset=false
    Write Material Random=true
   ) {module=InitializeParticle001}
   /SetVariables(Particles.CoronaSpriteSize=&Vector2DFromFloat(Value=&Multiply_Float(A=110, B=$System.ModifySize))) {isSetParameters=true}
   &CameraOffset(Camera Offset Amount=15)
   &InitialMeshOrientation(
    Rotation=&UniformRangedVector(Minimum=<-1 -1 -1>, Maximum=<1 1 1>)
    Use Orientation Vector=true
    Use Rotation Vector=true
   ) {module=InitialMeshOrientation001}
   /SetVariables(Particles.SubImageIndexAlt:NiagaraFloat=&RandomRangeFloat(Minimum=0, Maximum=3)) {isSetParameters=true}
   /SetVariables(Particles.Impact_Yellow_Orientation:Quat4f=/Niagara/DynamicInputs/Rotation/RandomQuaternion()) {isSetParameters=true}
  ]
  particleUpdateScript [
   &ParticleState()
   &ScaleSpriteSize(
    Scale Sprite Size Mode=::NewEnumerator0
    Uniform Scale Factor=&Multiply_Float(
     A=1
     B=&FloatFromCurve(FloatCurve=DataInterface() {propertyValues {Curve {keys [
          {arriveTangent=1.0961, interpMode=RCIM_Cubic, leaveTangent=1.0961, tangentMode=RCTM_User, time=0, value=0.75}
          {arriveTangent=0.2002, interpMode=RCIM_Cubic, leaveTangent=0.2002, tangentMode=RCTM_User, time=0.8, value=1.25}
          <at 1 0 RCIM_Cubic>
         ]}}})
    )
   )
   &DynamicMaterialParameters(
    Index 0 Param 1=&FloatFromCurve(FloatCurve=DataInterface() {propertyValues {Curve {keys [<at 0 1 RCIM_Cubic>,<at 0.8 0 RCIM_Cubic>]}}})
    Index 0 Param 2=2
    Index 0 Param 3=1
    Index 0 Param 4=1
    Index 1 Param 1=10
    Index 1 Param 2=0.4
    Index 1 Param 3=0.6
    Index 1 Param 4=1
    Index 2 Param 1=&FloatFromCurve(FloatCurve=DataInterface() {propertyValues {Curve {keys [<at 0.46 0>,<at 1 1>]}}})
    Index 2 Param 2=45
    Module.Param0WriteEnabled=true
    Module.Param0WriteEnabled1=true
    Module.Param0WriteEnabled2=true
    Module.Param1WriteEnabled=true
    Module.Param1WriteEnabled1=true
    Module.Param1WriteEnabled2=true
    Module.Param2WriteEnabled=true
    Module.Param2WriteEnabled1=true
    Module.Param2WriteEnabled2=true
    Module.Param3WriteEnabled=true
    Module.Param3WriteEnabled1=true
    Module.Param3WriteEnabled2=true
    Write Parameter Index 1=true
    Write Parameter Index 2=true
   )
   /SetVariables(Particles.MainGlowSize:Vector2f=&Multiply_Vector2D_ByFloat(Float=0.8, Vector2D=$Particles.SpriteSize)) {isSetParameters=true}
   /SetVariables(Particles.ImpactColor:LinearColor=&MultiplyLinearColor(
     Linear Color=&MultiplyLinearColor(
      Linear Color=<rgba 1.4419 0.3773 3 1>
      Scale Factor=&MakeVector4_FromLinearColor(LinearColor=&ColorFromCurve(ColorCurve=DataInterface() {propertyValues {
          AlphaCurve {keys [<at 0 1 RCIM_Cubic>,<at 1 0 RCIM_Cubic>]}
          BlueCurve {keys [<at 0.046 10 RCIM_Cubic>,<at 0.3368 1>,<at 0.8774 0.2969>]}
          GreenCurve {keys [<at 0.046 10 RCIM_Cubic>,<at 0.3368 1>,<at 0.8774 0.2969>]}
          RedCurve {keys [<at 0.046 10 RCIM_Cubic>,<at 0.3368 1>,<at 0.8774 0.2969>]}
         }}))
     )
     Scale Factor=&MakeVector4_FromLinearColor(LinearColor=$System.TODColorMultiplier)
    )) {isSetParameters=true}
   /SetVariables(Particles.CoronaSpriteColor:LinearColor=&MakeLinearColorFromVectorAndFloat(
     "Vector (RGB)"=&Multiply_Vector(
      A=&MakeVectorFromLinearColorRGB(Color=<rgba 0.6821 0.3298 1 1>)
      B=&MakeVectorFromLinearColorRGB(Color=$System.TODColorMultiplier)
     )
     "Float (Alpha)"=&Multiply_Float(
      A=&FloatFromCurve(FloatCurve=DataInterface() {propertyValues {Curve {keys [{arriveTangent=-4.1666, interpMode=RCIM_Cubic, leaveTangent=-4.1666, tangentMode=RCTM_User, time=0.35, value=1.5}, <at 0.95 0>]}}})
      B=&MakeFloatFromLinearColor(LinearColor=$System.TODColorMultiplier, Channel:ENiagaraLinearColor_Channels=::NewEnumerator3)
     )
    )) {isSetParameters=true}
   /SetVariables(Particles.CoronaSpriteSize=&Multiply_Vector2D_ByFloat(
     Float=$System.ModifySize
     Vector2D=&Multiply_Vector2D_ByFloat(
      Float=170
      Vector2D=&Vector2DFromFloat(Value=&FloatFromCurve(FloatCurve=DataInterface() {propertyValues {Curve {keys [
            {arriveTangent=2.2265, interpMode=RCIM_Cubic, leaveTangent=2.2265, tangentMode=RCTM_User, time=0, value=0}
            {interpMode=RCIM_Cubic, tangentMode=RCTM_User, time=1, value=1}
           ]}}}))
     )
    )) {isSetParameters=true}
   /SetVariables(Particles.CoronaDynMat:Vector4f=<1 2 1 0>) {isSetParameters=true}
   /SetVariables(
    Particles.DarkGlowColor:LinearColor=&ColorFromCurve(ColorCurve=DataInterface() {propertyValues {
       AlphaCurve {keys [<at 0 1 RCIM_Cubic>,<at 1 0 RCIM_Cubic>]}
       BlueCurve {keys [<at 0.07 0.3698 RCIM_Cubic>,<at 1 0.249 RCIM_Cubic>]}
       GreenCurve {keys [<at 0.07 0 RCIM_Cubic>,<at 1 0 RCIM_Cubic>]}
       RedCurve {keys [<at 0.07 0.1409 RCIM_Cubic>,<at 1 0.0405 RCIM_Cubic>]}
      }})
    Particles.DarkGlowSize:Vector2f=&Multiply_Vector2D_ByFloat(Float=0.45, Vector2D=$Particles.SpriteSize)
   ) {isSetParameters=true}
   &ScaleColor(
    Scale Alpha=&Multiply_Float(
     A=&FloatFromCurve(FloatCurve=DataInterface() {propertyValues {Curve {keys [<at 0 1>,<at 1 0>]}}})
     B=&MakeFloatFromLinearColor(LinearColor=$System.TODColorMultiplier, Channel:ENiagaraLinearColor_Channels=::NewEnumerator3)
    )
    Scale RGB=&Multiply_VectorByFloat(
     Vector=&MakeVectorFromLinearColorRGB(Color=$System.TODColorMultiplier)
     Float=&FNTimeOfDayScalar(Night Scale=1, Morning Scale=1, Evening Scale=1)
    )
    ScaleA=true
    ScaleRGB=true
   )
   &ScaleMeshSize(Scale Factor=&VectorFromFloat(Value=&Multiply_Float(
      A=$System.ModifySize
      B=&FloatFromCurve(FloatCurve=DataInterface() {propertyValues {Curve {keys [<at 0 0.35 RCIM_Cubic>,<at 1 1.25>]}}})
     )))
   &FN_Mesh_Attributes(
    Color=&MultiplyLinearColor(
     Linear Color=&MultiplyLinearColor(
      Linear Color=<rgba 0.7204 2 0.9206 1>
      Scale Factor=&MakeVector4_FromLinearColor(LinearColor=&ColorFromCurve(ColorCurve=DataInterface() {propertyValues {
          AlphaCurve {keys [<at 0.5626 1 RCIM_Cubic>,<at 1 0 RCIM_Cubic>]}
          BlueCurve {keys [<at 0.046 10 RCIM_Cubic>,<at 0.3368 1>,<at 0.8774 0.2969>]}
          GreenCurve {keys [<at 0.046 10 RCIM_Cubic>,<at 0.3368 1>,<at 0.8774 0.2969>]}
          RedCurve {keys [<at 0.046 10 RCIM_Cubic>,<at 0.3368 1>,<at 0.8774 0.2969>]}
         }}))
     )
     Scale Factor=&MakeVector4_FromLinearColor(LinearColor=$System.TODColorMultiplier)
    )
    Create Color Attribute=true
    Create Scale Attribute=true
    Scale=&Multiply_Vector(
     A=&VectorFromFloat(Value=$System.ModifySize)
     B=&Multiply_Vector(
      A=&VectorFromFloat(Value=&RandomRangeFloat(Minimum=1.3, Maximum=1.4))
      B=&VectorFromFloat(Value=&FloatFromCurve(FloatCurve=DataInterface() {propertyValues {Curve {keys [<at 0 0.35 RCIM_Cubic>,<at 1 1.25>]}}}))
     )
    )
    Write to Module Namespace=true
   )
  ]
  rendererClasses [&NiagaraSpriteRendererProperties, &NiagaraMeshRendererProperties]
  renderers [
   {rendererClass=&NiagaraSpriteRendererProperties}
   {rendererClass=&NiagaraSpriteRendererProperties, rendererIndex=1}
   {rendererClass=&NiagaraMeshRendererProperties, rendererIndex=2}
   {rendererClass=&NiagaraMeshRendererProperties, rendererIndex=3}
   {rendererClass=&NiagaraSpriteRendererProperties, rendererIndex=4}
  ]
 }
 Sparks {
  emitterUpdateScript [
   &EmitterState()
   &SpawnBurst_Instantaneous(Spawn Count=11, Use Spawn Probability=false, UseLoopCountLimit=false)
  ]
  particleSpawnScript [
   &InitializeParticle(
    Color=&SelectLinearColorFromArray(
     Array Sampling Mode=::NewEnumerator0
     Direct Array Index=?
     Color Selection Array=DataInterface() {propertyValues {ColorData [<rgba 0.5478 5 1.943 1>, <rgba 2.2679 0.2335 5 1>, <rgba 2.0967 5 3.1338 1>, <rgba 3.1607 1.6846 5 1>]}}
     Random Seed=0
    )
    Lifetime Max=0.5
    Lifetime Min=0.18
    Lifetime Mode=::NewEnumerator1
    Mass Mode=::NewEnumerator3
    Sprite Rotation Mode=::NewEnumerator2
    Sprite Size Mode=::NewEnumerator1
    Uniform Sprite Size Max=12
    Uniform Sprite Size Min=4
    UsePositionOffset=false
    Write Material Random=true
   )
   &CameraOffset(Camera Offset Amount=15)
   &ShapeLocation(MaskSpawnGroup=false, Override Seed=false, Sphere Radius=8, UseOwnerScale=true)
   &AddVelocity(
    Cone Angle=60
    Cone Axis=<0 0 1>
    Distribution Along Cone Axis=0.25
    Module.UseConeFalloff=true
    Module.UseVelDistribution=true
    Override Seed=false
    Speed Falloff From Cone Axis=0.5
    Velocity Mode=::NewEnumerator2
    Velocity Speed=&RandomRangeFloat(Minimum=400, Maximum=700)
   )
  ]
  particleUpdateScript [
   &ParticleState()
   &GravityForce()
   &Drag(Drag=&RandomRangeFloat(Minimum=4, Maximum=6), Use Linear Drag=true, Use Rotational Drag=false)
   &ScaleColor(
    Scale Alpha=&FloatFromCurve(FloatCurve=DataInterface() {propertyValues {Curve {keys [<at -0.0046 1.0113 RCIM_Cubic>,<at 1.0011 0.002 RCIM_Cubic>]}}})
    ScaleA=true
    ScaleRGB=true
   )
   &SolveForcesAndVelocity(Clamp Velocity=false, Limit Acceleration=false)
   &ScaleSpriteSizeBySpeed(Max Scale Factor=<1 2>, Min Scale Factor=<0.5 1>)
  ]
  rendererClasses [&NiagaraSpriteRendererProperties]
  renderers [{rendererClass=&NiagaraSpriteRendererProperties}]
 }
 Streak {
  emitterUpdateScript [
   &EmitterState(Inactive Response=::NewEnumerator1, Life Cycle Mode=::NewEnumerator1, Loop Behavior=::NewEnumerator1, Loop Duration=1, UseLoopDelay=false)
   &SpawnBurst_Instantaneous(Spawn Count=6, Use Spawn Probability=false, UseLoopCountLimit=false)
  ]
  particleSpawnScript [
   &InitializeParticle(
    Color Minimum=<rgba 0.4583 0.4583 0.4583 2>
    Color Mode=::NewEnumerator2
    Lifetime Max=0.5
    Lifetime Min=0.3
    Lifetime Mode=::NewEnumerator1
    Sprite Size Max=<15 50>
    Sprite Size Min=<5 30>
    Sprite Size Mode=::NewEnumerator2
    Sprite UV Mode=::NewEnumerator1
    UsePositionOffset=false
    Write Material Random=true
   )
   &CameraOffset(Camera Offset Amount=15)
   &ShapeLocation(
    Handle Radius=2
    Large Radius=2
    MaskSpawnGroup=false
    Offset Mode=::NewEnumerator0
    Override Seed=false
    Rotation Mode=::NewEnumerator1
    Shape Primitive=::NewEnumerator3
    UseOwnerScale=true
   )
   &AddVelocity(
    Cone Angle=80
    Cone Axis=<0 0 2>
    Module.UseConeFalloff=false
    Module.UseVelDistribution=false
    Override Seed=false
    Velocity Mode=::NewEnumerator2
    Velocity Speed=&RandomRangeFloat(Minimum=50, Maximum=100)
   ) {module=AddVelocity002}
  ]
  particleUpdateScript [
   &ParticleState()
   &ScaleSpriteSize(
    Non-Uniform Curve Scale=&RandomRangeVector2D(Minimum=<0.7 1.5>, Maximum=<0.9 2.5>)
    Non-Uniform Curve Sprite Scale=DataInterface() {propertyValues {
      XCurve {keys [
        {arriveTangent=0.8609, interpMode=RCIM_Cubic, leaveTangent=0.8609, tangentMode=RCTM_User, time=0, value=0.6}
        {arriveTangent=0.5103, interpMode=RCIM_Cubic, leaveTangent=0.5103, tangentMode=RCTM_User, time=0.3, value=0.8}
        {arriveTangent=0.0224, interpMode=RCIM_Cubic, leaveTangent=0.0224, tangentMode=RCTM_User, time=1, value=1}
       ]}
      YCurve {keys [
        {arriveTangent=0.8801, interpMode=RCIM_Cubic, leaveTangent=0.8801, tangentMode=RCTM_User, time=0, value=0.3}
        {arriveTangent=0.6586, interpMode=RCIM_Cubic, leaveTangent=0.6586, tangentMode=RCTM_User, time=0.2, value=0.85}
        {arriveTangent=-0.4778, interpMode=RCIM_Cubic, leaveTangent=-0.4778, tangentMode=RCTM_User, time=1, value=0.75}
       ]}
     }}
    Scale Sprite Size Mode=::NewEnumerator3
   ) {module=ScaleSpriteSize001}
   &ScaleColor(
    Scale Alpha=&FloatFromCurve(
     FloatCurve=DataInterface() {propertyValues {Curve {keys [
         <at 0 0 RCIM_Cubic>
         {arriveTangent=-1.2233, interpMode=RCIM_Cubic, leaveTangent=-1.2233, tangentMode=RCTM_User, time=0.1, value=1}
         <at 1 0 RCIM_Cubic>
        ]}}}
     Scale Curve=&Multiply_Float(A=1, B=&FNTimeOfDayScalar(Night Scale=0.6, Morning Scale=0.8, Evening Scale=0.6))
    )
    Scale RGB=&MakeVectorFromLinearColorRGB(Color=&ColorFromCurve(ColorCurve=DataInterface() {propertyValues {
        AlphaCurve {keys []}
        BlueCurve {keys [<at 0.1572 0.6071 RCIM_Cubic>,<at 0.3517 1>,<at 0.7153 0.0573>]}
        GreenCurve {keys [<at 0.1572 1 RCIM_Cubic>,<at 0.3517 0.1135>,<at 0.7153 0.0219>]}
        RedCurve {keys [<at 0.1572 0.251 RCIM_Cubic>,<at 0.3517 0.4095>,<at 0.7153 0.0308>]}
       }}))
    ScaleA=true
    ScaleRGB=true
   )
   &Drag(Drag=2.5, Use Linear Drag=true, Use Rotational Drag=true)
   &SolveForcesAndVelocity(Clamp Velocity=false, Limit Acceleration=false)
   &DynamicMaterialParameters(
    Index 0 Param 1=&FloatFromCurve(FloatCurve=DataInterface() {propertyValues {Curve {keys [
         {arriveTangent=-0.1846, time=0, value=0}
         {arriveTangent=1.3385, interpMode=RCIM_Cubic, leaveTangent=1.1594, tangentMode=RCTM_Break, time=0.5, value=0}
         {arriveTangent=2, interpMode=RCIM_Cubic, leaveTangent=2, time=0.8, value=0.5}
         {arriveTangent=3.0091, interpMode=RCIM_Cubic, leaveTangent=3.0091, tangentMode=RCTM_User, time=1, value=1}
        ]}}})
    Index 0 Param 2=&FloatFromCurve(FloatCurve=DataInterface() {propertyValues {Curve {keys [
         {arriveTangent=-0.0189, interpMode=RCIM_Cubic, leaveTangent=-0.0189, tangentMode=RCTM_User, time=0, value=1}
         {arriveTangent=-3.722, interpMode=RCIM_Cubic, leaveTangent=-3.722, tangentMode=RCTM_User, time=0.8, value=0}
        ]}}})
    Index 0 Param 4=&FloatFromCurve(FloatCurve=DataInterface() {propertyValues {Curve {keys [<at 0 0 RCIM_Cubic>,<at 1 1 RCIM_Cubic>]}}}, Scale Curve=0.8)
    Module.Index0Param0Write=true
    Module.Index0Param1Write=true
    Module.Index0Param2Write=true
    Module.Index0Param3Write=true
   )
  ]
  rendererClasses [&NiagaraSpriteRendererProperties]
  renderers [{rendererClass=&NiagaraSpriteRendererProperties}]
 }
}
summary {
 systemName=NS_Emote_VigorKnack_Impact
 emitters [
  {emitterName=Streak, rendererClasses [&NiagaraSpriteRendererProperties]}
  {
   emitterName=NE_Impact_Flare_and_Corona
   rendererClasses [&NiagaraSpriteRendererProperties, &NiagaraMeshRendererProperties]
  }
  {emitterName=Sparks, rendererClasses [&NiagaraSpriteRendererProperties]}
 ]
 userVariables [User.bIsFrontEndPreview {type=NiagaraBool}]
}
