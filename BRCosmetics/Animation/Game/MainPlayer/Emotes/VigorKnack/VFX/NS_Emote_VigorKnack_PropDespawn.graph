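# Niagara one-shot burst effect for the prop despawn in the VigorKnack emote, colored mint green and purple.
# A very short flare-and-corona emitter layers sprites and meshes, with curves driving growth and fade, scaled by system size and time-of-day color multipliers.
# Pop animates two mesh spike rings through per-renderer color, size and material attributes.
# Smoke and sparkle emitters, apparently lightweight, show no modules or renderers here.
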NiagaraSystem /BRCosmetics/Animation/Game/MainPlayer/Emotes/VigorKnack/VFX/NS_Emote_VigorKnack_PropDespawn
dependencies {
 usedDataInterfaces [&NiagaraDataInterfaceCurve, &NiagaraDataInterfaceColorCurve]
 usedDynamicInputs []
 usedModules [
  &self:SystemSpawnScript.SystemScriptSource.SystemScriptGraph.NiagaraNodeAssignment_0.SetVariables_#1
  &self:SystemSpawnScript.SystemScriptSource.SystemScriptGraph.NiagaraNodeAssignment_1.SetVariables_#2
  &SystemState
 ]
}
emitters {
 NE_Impact_Flare_and_Corona {
  emitterUpdateScript [
   &EmitterState()
   &SpawnBurst_Instantaneous(Spawn Count=1, Use Spawn Probability=false)
  ]
  particleSpawnScript [
   &InitializeParticle(
    Color=<rgba 0.6387 1 0.786 1>
    Lifetime=0.19
    Mesh Scale Mode=::NewEnumerator3
    Mesh Uniform Scale=1.75
    Sprite Size Mode=::NewEnumerator3
    Uniform Sprite Size=&Multiply_Float(A=245, B=$System.ModifySize)
    UsePositionOffset=false
    Write Material Random=true
   ) {module=InitializeParticle001}
   /SetVariables(Particles.CoronaSpriteSize=&Vector2DFromFloat(Value=&Multiply_Float(A=110, B=$System.ModifySize))) {isSetParameters=true}
   &CameraOffset(Camera Offset Amount=15)
   &InitialMeshOrientation(
    Rotation=&UniformRangedVector(Minimum=<-1 -1 -1>, Maximum=<1 1 1>)
    Use Orientation Vector=true
    Use Rotation Vector=true
   ) {module=InitialMeshOrientation001}
   /SetVariables(Particles.SubImageIndexAlt:NiagaraFloat=&RandomRangeFloat(Minimum=0, Maximum=3)) {isSetParameters=true}
   /SetVariables(Particles.Impact_Yellow_Orientation:Quat4f=/Niagara/DynamicInputs/Rotation/RandomQuaternion()) {isSetParameters=true}
  ]
  particleUpdateScript [
   &ParticleState()
   &ScaleSpriteSize(
    Scale Sprite Size Mode=::NewEnumerator0
    Uniform Scale Factor=&Multiply_Float(
     A=1
     B=&FloatFromCurve(FloatCurve=DataInterface() {propertyValues {Curve {keys [
          {arriveTangent=1.0961, interpMode=RCIM_Cubic, leaveTangent=1.0961, tangentMode=RCTM_User, time=0, value=0.75}
          {arriveTangent=0.2002, interpMode=RCIM_Cubic, leaveTangent=0.2002, tangentMode=RCTM_User, time=0.8, value=1.25}
          <at 1 0 RCIM_Cubic>
         ]}}})
    )
   )
   &DynamicMaterialParameters(
    Index 0 Param 1=&FloatFromCurve(FloatCurve=DataInterface() {propertyValues {Curve {keys [<at 0 1 RCIM_Cubic>,<at 0.8 0 RCIM_Cubic>]}}})
    Index 0 Param 2=2
    Index 0 Param 3=1
    Index 0 Param 4=1
    Index 1 Param 1=10
    Index 1 Param 2=0.4
    Index 1 Param 3=0.6
    Index 1 Param 4=1
    Index 2 Param 1=&FloatFromCurve(FloatCurve=DataInterface() {propertyValues {Curve {keys [<at 0.46 0>,<at 1 1>]}}})
    Index 2 Param 2=45
    Module.Param0WriteEnabled=true
    Module.Param0WriteEnabled1=true
    Module.Param0WriteEnabled2=true
    Module.Param1WriteEnabled=true
    Module.Param1WriteEnabled1=true
    Module.Param1WriteEnabled2=true
    Module.Param2WriteEnabled=true
    Module.Param2WriteEnabled1=true
    Module.Param2WriteEnabled2=true
    Module.Param3WriteEnabled=true
    Module.Param3WriteEnabled1=true
    Module.Param3WriteEnabled2=true
    Write Parameter Index 1=true
    Write Parameter Index 2=true
   )
   /SetVariables(Particles.MainGlowSize:Vector2f=&Multiply_Vector2D_ByFloat(Float=0.8, Vector2D=$Particles.SpriteSize)) {isSetParameters=true}
   /SetVariables(Particles.ImpactColor:LinearColor=&MultiplyLinearColor(
     Linear Color=&MultiplyLinearColor(
      Linear Color=<rgba 1.4419 0.3773 3 1>
      Scale Factor=&MakeVector4_FromLinearColor(LinearColor=&ColorFromCurve(ColorCurve=DataInterface() {propertyValues {
          AlphaCurve {keys [<at 0 1 RCIM_Cubic>,<at 1 0 RCIM_Cubic>]}
          BlueCurve {keys [<at 0.046 10 RCIM_Cubic>,<at 0.3368 1>,<at 0.8774 0.2969>]}
          GreenCurve {keys [<at 0.046 10 RCIM_Cubic>,<at 0.3368 1>,<at 0.8774 0.2969>]}
          RedCurve {keys [<at 0.046 10 RCIM_Cubic>,<at 0.3368 1>,<at 0.8774 0.2969>]}
         }}))
     )
     Scale Factor=&MakeVector4_FromLinearColor(LinearColor=$System.TODColorMultiplier)
    )) {isSetParameters=true}
   /SetVariables(Particles.CoronaSpriteColor:LinearColor=&MakeLinearColorFromVectorAndFloat(
     "Vector (RGB)"=&Multiply_Vector(
      A=&MakeVectorFromLinearColorRGB(Color=<rgba 0.6821 0.3298 1 1>)
      B=&MakeVectorFromLinearColorRGB(Color=$System.TODColorMultiplier)
     )
     "Float (Alpha)"=&Multiply_Float(
      A=&FloatFromCurve(FloatCurve=DataInterface() {propertyValues {Curve {keys [{arriveTangent=-4.1666, interpMode=RCIM_Cubic, leaveTangent=-4.1666, tangentMode=RCTM_User, time=0.35, value=1.5}, <at 0.95 0>]}}})
      B=&MakeFloatFromLinearColor(LinearColor=$System.TODColorMultiplier, Channel:ENiagaraLinearColor_Channels=::NewEnumerator3)
     )
    )) {isSetParameters=true}
   /SetVariables(Particles.CoronaSpriteSize=&Multiply_Vector2D_ByFloat(
     Float=$System.ModifySize
     Vector2D=&Multiply_Vector2D_ByFloat(
      Float=170
      Vector2D=&Vector2DFromFloat(Value=&FloatFromCurve(FloatCurve=DataInterface() {propertyValues {Curve {keys [
            {arriveTangent=2.2265, interpMode=RCIM_Cubic, leaveTangent=2.2265, tangentMode=RCTM_User, time=0, value=0}
            {interpMode=RCIM_Cubic, tangentMode=RCTM_User, time=1, value=1}
           ]}}}))
     )
    )) {isSetParameters=true}
   /SetVariables(Particles.CoronaDynMat:Vector4f=<1 2 1 0>) {isSetParameters=true}
   /SetVariables(
    Particles.DarkGlowColor:LinearColor=&ColorFromCurve(ColorCurve=DataInterface() {propertyValues {
       AlphaCurve {keys [<at 0 1 RCIM_Cubic>,<at 1 0 RCIM_Cubic>]}
       BlueCurve {keys [<at 0.07 0.3698 RCIM_Cubic>,<at 1 0.249 RCIM_Cubic>]}
       GreenCurve {keys [<at 0.07 0 RCIM_Cubic>,<at 1 0 RCIM_Cubic>]}
       RedCurve {keys [<at 0.07 0.1409 RCIM_Cubic>,<at 1 0.0405 RCIM_Cubic>]}
      }})
    Particles.DarkGlowSize:Vector2f=&Multiply_Vector2D_ByFloat(Float=0.45, Vector2D=$Particles.SpriteSize)
   ) {isSetParameters=true}
   &ScaleColor(
    Scale Alpha=&Multiply_Float(
     A=&FloatFromCurve(FloatCurve=DataInterface() {propertyValues {Curve {keys [<at 0 1>,<at 1 0>]}}})
     B=&MakeFloatFromLinearColor(LinearColor=$System.TODColorMultiplier, Channel:ENiagaraLinearColor_Channels=::NewEnumerator3)
    )
    Scale RGB=&Multiply_VectorByFloat(
     Vector=&MakeVectorFromLinearColorRGB(Color=$System.TODColorMultiplier)
     Float=&FNTimeOfDayScalar(Night Scale=1, Morning Scale=1, Evening Scale=1)
    )
    ScaleA=true
    ScaleRGB=true
   )
   &ScaleMeshSize(Scale Factor=&VectorFromFloat(Value=&Multiply_Float(
      A=$System.ModifySize
      B=&FloatFromCurve(FloatCurve=DataInterface() {propertyValues {Curve {keys [<at 0 0.35 RCIM_Cubic>,<at 1 1.25>]}}})
     )))
   &FN_Mesh_Attributes(
    Color=&MultiplyLinearColor(
     Linear Color=&MultiplyLinearColor(
      Linear Color=<rgba 0.7204 2 0.9206 1>
      Scale Factor=&MakeVector4_FromLinearColor(LinearColor=&ColorFromCurve(ColorCurve=DataInterface() {propertyValues {
          AlphaCurve {keys [<at 0.5626 1 RCIM_Cubic>,<at 1 0 RCIM_Cubic>]}
          BlueCurve {keys [<at 0.046 10 RCIM_Cubic>,<at 0.3368 1>,<at 0.8774 0.2969>]}
          GreenCurve {keys [<at 0.046 10 RCIM_Cubic>,<at 0.3368 1>,<at 0.8774 0.2969>]}
          RedCurve {keys [<at 0.046 10 RCIM_Cubic>,<at 0.3368 1>,<at 0.8774 0.2969>]}
         }}))
     )
     Scale Factor=&MakeVector4_FromLinearColor(LinearColor=$System.TODColorMultiplier)
    )
    Create Color Attribute=true
    Create Scale Attribute=true
    Scale=&Multiply_Vector(
     A=&VectorFromFloat(Value=$System.ModifySize)
     B=&Multiply_Vector(
      A=&VectorFromFloat(Value=&RandomRangeFloat(Minimum=1.3, Maximum=1.4))
      B=&VectorFromFloat(Value=&FloatFromCurve(FloatCurve=DataInterface() {propertyValues {Curve {keys [<at 0 0.35 RCIM_Cubic>,<at 1 1.25>]}}}))
     )
    )
    Write to Module Namespace=true
   )
  ]
  rendererClasses [&NiagaraSpriteRendererProperties, &NiagaraMeshRendererProperties]
  renderers [
   {rendererClass=&NiagaraSpriteRendererProperties}
   {rendererClass=&NiagaraSpriteRendererProperties, rendererIndex=1}
   {rendererClass=&NiagaraMeshRendererProperties, rendererIndex=2}
   {rendererClass=&NiagaraMeshRendererProperties, rendererIndex=3}
   {rendererClass=&NiagaraSpriteRendererProperties, rendererIndex=4}
  ]
 }
 Pop {
  emitterUpdateScript [
   &EmitterState()
   &SpawnBurst_Instantaneous(Spawn Group=2, Use Spawn Probability=false, UseLoopCountLimit=false) {module=SpawnBurst_Instantaneous002}
  ]
  particleSpawnScript [
   /SetVariables(Particles.VisibilityTag=&SpawnGroup()) {isSetParameters=true}
   &InitializeParticle(
    Lifetime=&MergeHelper_FloatSwitch(0=0.07, 1=0.2, 2=0.2, 3=0.15, 4=0, Index=&SpawnGroup(), Num=::NewEnumerator3)
    Position Offset=$System.BurstPosition
    Sprite Rotation Angle=&MergeHelper_FloatSwitch(0=0, 1=0, 2=25, 3=0, 4=0, Index=&SpawnGroup(), Num=::NewEnumerator3)
    Sprite Rotation Mode=::NewEnumerator0
    Sprite Size Mode=::NewEnumerator3
    Uniform Sprite Size=15
    UsePositionOffset=true
    Write Material Random=true
   )
   &DynamicMaterialParameters(
    Index 0 Param 4=1
    Module.Index0Float0Write=true
    Module.Index0Float1Write=true
    Module.Index0Float2Write=true
    Module.Index0Float3Write=true
   )
  ]
  particleUpdateScript [
   &ParticleState()
   &SolveForcesAndVelocity(Clamp Velocity=false, Limit Acceleration=false)
   /SetVariables(Particles.Ring Dynamic Param:Vector4f=&MakeVector4(
     X=&FloatFromCurve(FloatCurve=DataInterface() {propertyValues {Curve {keys [
          {arriveTangent=-7.2529, interpMode=RCIM_Cubic, leaveTangent=-7.2529, tangentMode=RCTM_User, time=-0.0022, value=1.0993}
          {arriveTangent=0.875, interpMode=RCIM_Cubic, leaveTangent=0.875, tangentMode=RCTM_User, time=0.5, value=0}
          {arriveTangent=3.8908, interpMode=RCIM_Cubic, leaveTangent=3.8908, tangentMode=RCTM_User, time=1, value=1}
         ]}}})
     Y=&FloatFromCurve(FloatCurve=DataInterface() {propertyValues {Curve {keys [
          {arriveTangent=5.2096, interpMode=RCIM_Cubic, leaveTangent=5.2096, tangentMode=RCTM_User, time=0, value=0.5}
          {arriveTangent=0.2132, interpMode=RCIM_Cubic, leaveTangent=0.2132, tangentMode=RCTM_User, time=1, value=3}
         ]}}})
     Z=&FloatFromCurve(FloatCurve=DataInterface() {propertyValues {Curve {
         defaultValue=1
         keys [
          {arriveTangent=0.0994, interpMode=RCIM_Cubic, leaveTangent=0.0994, tangentMode=RCTM_User, time=0, value=1}
          {arriveTangent=-1.6508, interpMode=RCIM_Cubic, leaveTangent=-1.6508, tangentMode=RCTM_User, time=1, value=0}
         ]
        }}})
     W=&FloatFromCurve(
      FloatCurve=DataInterface() {propertyValues {Curve {keys [
          {arriveTangent=2.3563, interpMode=RCIM_Cubic, leaveTangent=2.3563, tangentMode=RCTM_User, time=0, value=0}
          {tangentMode=RCTM_User, time=1, value=1}
         ]}}}
      Scale Curve=2
     )
    )) {isSetParameters=true}
   /SetVariables(
    Particles.Spike Ring Color:LinearColor=&MakeLinearColorFromVectorAndFloat(
     "Vector (RGB)"=&MakeVectorFromLinearColorRGB(Color=<rgba 2.1657 5 2.454 1>)
     "Float (Alpha)"=&FloatFromCurve(FloatCurve=DataInterface() {propertyValues {Curve {
         defaultValue=1
         keys [{interpMode=RCIM_Cubic, tangentMode=RCTM_User, time=0, value=1}, {interpMode=RCIM_Cubic, tangentMode=RCTM_User, time=1, value=0}]
        }}})
    )
    Particles.Spike Ring Size:Vector3f=&Multiply_VectorByFloat(
     Vector=<0.4 0.4 0.4>
     Float=&FloatFromCurve(
      FloatCurve=DataInterface() {propertyValues {Curve {defaultValue=0, keys [{tangentMode=RCTM_User, time=0, value=0}, <at 0.1024 0.2862>, {tangentMode=RCTM_User, time=1, value=1}]}}}
      Scale Curve=0.85
     )
    )
   ) {isSetParameters=true}
   /SetVariables(
    Particles.Spike Ring Color 1:LinearColor=&MakeLinearColorFromVectorAndFloat(
     "Vector (RGB)"=&MakeVectorFromLinearColorRGB(Color=<rgba 2.5218 1.0526 6 1>)
     "Float (Alpha)"=&FloatFromCurve(FloatCurve=DataInterface() {propertyValues {Curve {
         defaultValue=1
         keys [{interpMode=RCIM_Cubic, tangentMode=RCTM_User, time=0, value=1}, {interpMode=RCIM_Cubic, tangentMode=RCTM_User, time=1, value=0}]
        }}})
    )
    Particles.Spike Ring Size 1:Vector3f=&Multiply_VectorByFloat(
     Vector=<0.7 0.7 0.7>
     Float=&FloatFromCurve(
      FloatCurve=DataInterface() {propertyValues {Curve {defaultValue=0, keys [{tangentMode=RCTM_User, time=0, value=0}, <at 0.1024 0.2862>, {tangentMode=RCTM_User, time=1, value=1}]}}}
      Scale Curve=0.85
     )
    )
   ) {isSetParameters=true}
  ]
  rendererClasses [&NiagaraMeshRendererProperties]
  renderers [
   {rendererClass=&NiagaraMeshRendererProperties, rendererIndex=0}
   {rendererClass=&NiagaraMeshRendererProperties, rendererIndex=1}
  ]
 }
 SmokeLWE {rendererClasses [], renderers []}
 SparklesLWE {rendererClasses [], renderers []}
}
summary {
 systemName=NS_Emote_VigorKnack_PropDespawn
 emitters [
  {emitterName=SparklesLWE, rendererClasses []}
  {emitterName=SmokeLWE, rendererClasses []}
  {
   emitterName=NE_Impact_Flare_and_Corona
   rendererClasses [&NiagaraSpriteRendererProperties, &NiagaraMeshRendererProperties]
  }
  {emitterName=Pop, rendererClasses [&NiagaraMeshRendererProperties]}
 ]
 userVariables [
  User.bIsFrontEnd {type=NiagaraBool}
  User.bIsFrontEndPreview {type=NiagaraBool}
 ]
}
