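# Niagara swing effect for the VigorKnack emote, flashing a short-lived, pale cyan "Shing" sprite.
# Each change in User.BurstIndex triggers one spawn and selects that swing's offset, rotation and pivot pair from switch tables.
# A camera-vector dot product against the Y axis blends those values, adapting the sprite to viewing angle.
# Non-uniform size curves stretch and squash it, while alternate initialize and time-based pivot modules remain disabled.
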NiagaraSystem /BRCosmetics/Animation/Game/MainPlayer/Emotes/VigorKnack/VFX/NS_Emote_VigorKnack_Swing
dependencies {
 usedDataInterfaces [
  &NiagaraDataInterfaceVectorCurve
  &NiagaraDataInterfaceCamera
  &NiagaraDataInterfaceCurve
  &NiagaraDataInterfaceVector2DCurve
 ]
 usedDynamicInputs []
 usedModules [&SystemState]
}
emitters {Shing {
  emitterUpdateScript [
   &EmitterState()
   /Game/Effects/Niagara/Modules/Logic/UpdateValueStateChange(InputIntParam:NiagaraInt32=&MakeIntFromFloat(Float=$User.BurstIndex, Conversion Mode=::NewEnumerator1))
   &self:TriggerSpawn(
    Spawn Count:NiagaraInt32=&MergeHelper_IntSwitch(
     0=1
     1=1
     2=1
     3=1
     4=1
     5=1
     6=1
     Index=&MakeIntFromFloat(Float=$User.BurstIndex, Conversion Mode=::NewEnumerator1)
     Num=::NewEnumerator5
    )
    Spawn Group:NiagaraInt32=&MakeIntFromFloat(Float=0, Conversion Mode=::NewEnumerator1)
    Trigger:NiagaraBool=$Emitter.UpdateValueStateChange.Changed
   ) {module=TriggerSpawn}
  ]
  particleSpawnScript [
   &InitializeParticle(
    Color Mode=::NewEnumerator0
    Lifetime=0.35
    Position Offset=&Lerp_Vector(
     A=&MergeHelper_VectorSwitch(
      0=<0 38 105>
      1=<0 38 85>
      2=<5 38 95>
      3=<-5 38 90>
      4=<-7 38 95>
      5=<-5 38 75>
      6=<15 38 95>
      Index=&MakeIntFromFloat(Float=$User.BurstIndex, Conversion Mode=::NewEnumerator1)
      Num=::NewEnumerator5
     )
     B=&MergeHelper_VectorSwitch(
      0=<0 38 105>
      1=<0 38 85>
      2=<5 38 95>
      3=<-5 38 90>
      4=<-7 38 95>
      5=<-5 38 75>
      6=<15 38 75>
      Index=&MakeIntFromFloat(Float=$User.BurstIndex, Conversion Mode=::NewEnumerator1)
      Num=::NewEnumerator5
     )
     Alpha=&MakeFloatFromVector(
      Vector=&VectorFromCurve(
       VectorCurve=DataInterface() {propertyValues {XCurve {keys [<at 0 1>,<at 1 0>]}, YCurve {keys [<at 0 1>,<at 1 0>]}, ZCurve {keys [<at 0 1>,<at 1 0>]}}}
       CurveIndex=&DotProduct(
        Vector A=&ReturnCameraVector(
         Camera Vectors=::NewEnumerator4
         Coordinate Space=::Simulation
         Origin Position=$Particles.Position
         CameraQuery=DataInterface() {propertyValues {bRequireCurrentFrameData=false}}
        )
        Vector B=<0 1 0>
       )
      )
      Channel=::NewEnumerator0
     )
    )
    Sprite Rotation Angle=&Lerp_Float(
     A=&MergeHelper_FloatSwitch(
      0=3
      1=-33
      2=0
      3=-30
      4=0
      5=-35
      6=-30
      Index=&MakeIntFromFloat(Float=$User.BurstIndex, Conversion Mode=::NewEnumerator1)
      Num=::NewEnumerator5
     )
     B=&MergeHelper_FloatSwitch(
      0=5
      1=5
      2=5
      3=5
      4=5
      5=5
      6=20
      Index=&MakeIntFromFloat(Float=$User.BurstIndex, Conversion Mode=::NewEnumerator1)
      Num=::NewEnumerator5
     )
     Alpha=&MakeFloatFromVector(
      Vector=&VectorFromCurve(
       VectorCurve=DataInterface() {propertyValues {XCurve {keys [<at 0 1>,<at 1 0>]}, YCurve {keys [<at 0 1>,<at 1 0>]}, ZCurve {keys [<at 0 1>,<at 1 0>]}}}
       CurveIndex=&DotProduct(
        Vector A=&ReturnCameraVector(
         Camera Vectors=::NewEnumerator4
         Coordinate Space=::Simulation
         Origin Position=$Particles.Position
         CameraQuery=DataInterface() {propertyValues {bRequireCurrentFrameData=false}}
        )
        Vector B=<0 1 0>
       )
      )
      Channel=::NewEnumerator0
     )
    )
    Sprite Rotation Mode=::NewEnumerator0
    Sprite Size=<70 62>
    Sprite Size Mode=::NewEnumerator4
    UsePositionOffset=true
    Write Material Random=true
   )
   &InitializeParticle(
    Color Mode=::NewEnumerator0
    Lifetime=0.35
    Position Offset=&MergeHelper_VectorSwitch(
     0=<0 38 105>
     1=<0 38 85>
     2=<5 38 95>
     3=<-5 38 90>
     4=<-7 38 95>
     5=<-5 38 75>
     6=<15 38 95>
     7=<-10 38 75>
     Index=&MakeIntFromFloat(Float=$User.BurstIndex, Conversion Mode=::NewEnumerator1)
     Num=::NewEnumerator6
    )
    Sprite Rotation Angle=&Multiply_Float(
     A=&RemapRange(
      Input Value To Remap=&MakeFloatFromVector(
       Vector=&VectorFromCurve(
        VectorCurve=DataInterface() {propertyValues {XCurve {keys [<at 0 1>,<at 1 0>]}, YCurve {keys [<at 0 1>,<at 1 0>]}, ZCurve {keys [<at 0 1>,<at 1 0>]}}}
        CurveIndex=&DotProduct(
         Vector A=&ReturnCameraVector(
          Camera Vectors=::NewEnumerator4
          Coordinate Space=::Simulation
          Origin Position=$Particles.Position
          CameraQuery=DataInterface() {propertyValues {bRequireCurrentFrameData=false}}
         )
         Vector B=<0 1 0>
        )
       )
       Channel=::NewEnumerator0
      )
      Input Min=0.2
      Input Max=0.9
      Output Min=1
      Output Max=-1
     )
     B=&MergeHelper_FloatSwitch(
      0=3
      1=-33
      2=0
      3=-30
      4=0
      5=-35
      6=-30
      7=-45
      Index=&MakeIntFromFloat(Float=$User.BurstIndex, Conversion Mode=::NewEnumerator1)
      Num=::NewEnumerator6
     )
    )
    Sprite Rotation Mode=::NewEnumerator0
    Sprite Size=<70 62>
    Sprite Size Mode=::NewEnumerator4
    UsePositionOffset=true
    Write Material Random=true
   ) {enabled=false, module=InitializeParticle001}
  ]
  particleUpdateScript [
   &ParticleState()
   &ScaleSpriteSize(
    Non-Uniform Curve Sprite Scale=DataInterface() {propertyValues {
      XCurve {keys [
        {arriveTangent=-2.2384, interpMode=RCIM_Cubic, leaveTangent=-2.2384, tangentMode=RCTM_User, time=0.0096, value=1}
        {arriveTangent=-0.5926, interpMode=RCIM_Cubic, leaveTangent=-0.5926, tangentMode=RCTM_User, time=0.066, value=3}
        {arriveTangent=-0.7056, interpMode=RCIM_Cubic, leaveTangent=-0.7056, tangentMode=RCTM_User, time=0.135, value=3}
        {arriveTangent=-25.5317, interpMode=RCIM_Cubic, leaveTangent=-25.5317, time=0.154, value=1.1466}
        {arriveTangent=-0.575, interpMode=RCIM_Cubic, leaveTangent=-0.575, time=0.214, value=0.9831}
        {arriveTangent=-0.1265, interpMode=RCIM_Cubic, leaveTangent=-0.1265, time=0.4089, value=1}
        {arriveTangentWeight=0.1701, interpMode=RCIM_Cubic, leaveTangentWeight=0.0269, tangentMode=RCTM_User, time=0.9601, value=0.8887}
        {arriveTangent=-41.738, interpMode=RCIM_Cubic, leaveTangent=-41.738, tangentMode=RCTM_User, time=1, value=0}
       ]}
      YCurve {keys [
        <at 0 1 RCIM_Cubic>
        {arriveTangent=-3.1456, interpMode=RCIM_Cubic, leaveTangent=-3.1456, time=0.0327, value=0.5786}
        {arriveTangent=8.2936, interpMode=RCIM_Cubic, leaveTangent=8.2936, time=0.1261, value=0.6032}
        {arriveTangent=4.0779, interpMode=RCIM_Cubic, leaveTangent=4.0779, time=0.1938, value=1.9148}
        {arriveTangentWeight=0.0196, interpMode=RCIM_Cubic, leaveTangentWeight=0.3801, tangentMode=RCTM_User, time=0.2234, value=1}
        {arriveTangent=-0.1512, interpMode=RCIM_Cubic, leaveTangent=-0.1512, time=0.409, value=1}
        {arriveTangentWeight=0.3344, interpMode=RCIM_Cubic, leaveTangentWeight=0.0528, tangentMode=RCTM_User, time=0.9594, value=0.8887}
        <at 1 0 RCIM_Cubic>
       ]}
     }}
    Scale Sprite Size Mode=::NewEnumerator3
   )
   &ScaleColor(
    Scale Alpha=&FloatFromCurve(FloatCurve=DataInterface() {propertyValues {Curve {
        defaultValue=0.0636
        keys [
         <at 0.1065 1.0156 RCIM_Cubic>
         {arriveTangent=-1.7223, interpMode=RCIM_Cubic, leaveTangent=-1.7223, tangentMode=RCTM_User, time=0.3743, value=0.607}
         {arriveTangent=-0.1661, interpMode=RCIM_Cubic, leaveTangent=-0.1661, tangentMode=RCTM_User, time=1, value=0}
        ]
       }}})
    Scale RGB=&MakeVectorFromLinearColorRGB(Color=<rgba 0.4644 0.7659 0.7948 1>)
    ScaleA=false
    ScaleRGB=true
   )
   /SetVariables(Particles.PivotOffset:Vector2f=&Lerp_Vector2(
     A=&MergeHelper_Vector2Switch(
      0=<0.6 0.5>
      1=<0.4 0.5>
      2=<0.6 0.5>
      3=<0.4 0.5>
      4=<0.6 0.5>
      5=<0.4 0.5>
      6=<0.6 0.5>
      7=<0.4 0.5>
      Index=&MakeIntFromFloat(Float=$User.BurstIndex, Conversion Mode=::NewEnumerator1)
      Num=::NewEnumerator6
     )
     B=&MergeHelper_Vector2Switch(
      0=<0.25 0.5>
      1=<0.75 0.5>
      2=<0.25 0.5>
      3=<0.75 0.5>
      4=<0.25 0.5>
      5=<0.75 0.5>
      6=<0.25 0.5>
      7=<0.75 0.5>
      Index=&MakeIntFromFloat(Float=$User.BurstIndex, Conversion Mode=::NewEnumerator1)
      Num=::NewEnumerator6
     )
     Alpha=&FloatFromCurve(FloatCurve=DataInterface() {propertyValues {Curve {
         defaultValue=1
         keys [{interpMode=RCIM_Cubic, tangentMode=RCTM_User, time=0, value=0}, {interpMode=RCIM_Cubic, tangentMode=RCTM_User, time=0.2061, value=1}]
        }}})
    )) {enabled=false, isSetParameters=true}
   /SetVariables(Particles.PivotOffset:Vector2f=&Lerp_Vector2(
     A=&MergeHelper_Vector2Switch(
      0=<0.9 0.5>
      1=<0.1 0.5>
      2=<0.9 0.5>
      3=<0.1 0.5>
      4=<0.9 0.5>
      5=<0.1 0.5>
      6=<0.9 0.5>
      7=<0.1 0.5>
      Index=&MakeIntFromFloat(Float=$User.BurstIndex, Conversion Mode=::NewEnumerator1)
      Num=::NewEnumerator6
     )
     B=&MergeHelper_Vector2Switch(
      0=<0.1 0.5>
      1=<0.9 0.5>
      2=<0.1 0.5>
      3=<0.9 0.5>
      4=<0.1 0.5>
      5=<0.9 0.5>
      6=<0.1 0.5>
      7=<0.9 0.5>
      Index=&MakeIntFromFloat(Float=$User.BurstIndex, Conversion Mode=::NewEnumerator1)
      Num=::NewEnumerator6
     )
     Alpha=&RemapRange(
      Input Value To Remap=&MakeFloatFromVector(
       Vector=&VectorFromCurve(
        VectorCurve=DataInterface() {propertyValues {XCurve {keys [<at 0 1>,<at 1 0>]}, YCurve {keys [<at 0 1>,<at 1 0>]}, ZCurve {keys [<at 0 1>,<at 1 0>]}}}
        CurveIndex=&DotProduct(
         Vector A=&ReturnCameraVector(
          Camera Vectors=::NewEnumerator4
          Coordinate Space=::Simulation
          Origin Position=$Particles.Position
          CameraQuery=DataInterface() {propertyValues {bRequireCurrentFrameData=false}}
         )
         Vector B=<0 1 0>
        )
       )
       Channel=::NewEnumerator0
      )
      Input Min=0.2
      Input Max=0.9
      Output Min=0
      Output Max=1
     )
    )) {isSetParameters=true}
  ]
  rendererClasses [&NiagaraSpriteRendererProperties]
  renderers [{rendererClass=&NiagaraSpriteRendererProperties}]
 }}
summary {
 systemName=NS_Emote_VigorKnack_Swing
 emitters [{emitterName=Shing, rendererClasses [&NiagaraSpriteRendererProperties]}]
 userVariables [
  User.bIsFrontEnd {type=NiagaraBool}
  User.bIsFrontEndPreview {type=NiagaraBool}
  User.BurstIndex {type=NiagaraFloat, defaultValue=0, description=""}
 ]
}
